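# Sound cue for the footsteps of the Realm emote, layering step sounds with suit foley.
# A mixer blends a random pick of nine step waves with a quieter random suit-foley pick, whose pitch and volume a modulator varies.
# It uses emote attenuation and foley concurrency, but is classed as emote music and has a send entry for the music audio-analysis submix.
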SoundCue /BRCosmetics/Sounds/Emotes/Realm/Emote_Realm_Steps_Cue
root=Mixer_0
nodes [
 {id=Mixer_0, children [Random_0,Modulator_0], properties {inputVolume [1,0.45]}}
 {id=Modulator_0, children [Random_1], properties {pitchMax=1.1, pitchMin=1, volumeMax=1, volumeMin=0.6}}
 {
  id=Random_0
  children [WavePlayer_4, WavePlayer_5, WavePlayer_6, WavePlayer_7, WavePlayer_8, WavePlayer_9, WavePlayer_10, WavePlayer_12, WavePlayer_13]
  properties {weights [1,1,1,1,1,1,1,1,1]}
 }
 {
  id=Random_1
  children [WavePlayer_14, WavePlayer_15, WavePlayer_16, WavePlayer_17, WavePlayer_18, WavePlayer_19, WavePlayer_20]
  properties {weights [1,1,1,1,1,1,1]}
 }
 {id=WavePlayer_10, properties {soundWaveAssetPtr=&/Emote_Realm_Step_07}}
 {id=WavePlayer_12, properties {soundWaveAssetPtr=&/Emote_Realm_Step_08}}
 {id=WavePlayer_13, properties {soundWaveAssetPtr=&/Emote_Realm_Step_09}}
 {id=WavePlayer_14, properties {soundWaveAssetPtr=&/Emote_Realm_SuitFoley_01}}
 {id=WavePlayer_15, properties {soundWaveAssetPtr=&/Emote_Realm_SuitFoley_02}}
 {id=WavePlayer_16, properties {soundWaveAssetPtr=&/Emote_Realm_SuitFoley_03}}
 {id=WavePlayer_17, properties {soundWaveAssetPtr=&/Emote_Realm_SuitFoley_04}}
 {id=WavePlayer_18, properties {soundWaveAssetPtr=&/Emote_Realm_SuitFoley_05}}
 {id=WavePlayer_19, properties {soundWaveAssetPtr=&/Emote_Realm_SuitFoley_06}}
 {id=WavePlayer_20, properties {soundWaveAssetPtr=&/Emote_Realm_SuitFoley_07}}
 {id=WavePlayer_4, properties {soundWaveAssetPtr=&/Emote_Realm_Step_01}}
 {id=WavePlayer_5, properties {soundWaveAssetPtr=&/Emote_Realm_Step_02}}
 {id=WavePlayer_6, properties {soundWaveAssetPtr=&/Emote_Realm_Step_03}}
 {id=WavePlayer_7, properties {soundWaveAssetPtr=&/Emote_Realm_Step_04}}
 {id=WavePlayer_8, properties {soundWaveAssetPtr=&/Emote_Realm_Step_05}}
 {id=WavePlayer_9, properties {soundWaveAssetPtr=&/Emote_Realm_Step_06}}
]
properties {
 attenuationSettings=&Fort_Emotes_Athena_Att
 maxDistance=1600
 soundClassObject=&Fort_Music_Emote
 volumeMultiplier=0.6
 concurrencyOverrides {maxCount=3, resolutionRule=StopOldest}
 concurrencySet [/Game/Sounds/Concurrency_Rules/Athena_EmoteFoley_Conc]
 soundSubmixSends [{minSendLevel=0, soundSubmix=&Fortnite_AudioAnalysis_Music}]
}
